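; PCG subgraph that auto-splits a polygon with one generated cut path, outputting the cut pieces and the clipped splitting spline.
; It cleans the outline, anchors on the longest outer edge, rotates by 90 degrees plus random AngleJitter, and places a straight line according to Ratio.
; Allow_Disturbance optionally resamples that line at a set distance, randomly offsetting interior points while endpoints stay fixed.
; The chosen path cuts the original polygon and is clipped to it.
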(graph AutoSplitPolygon
  (params (Ratio :type Double :default 0.5) (SplitingWidth :type Double :default 1e+02) (AngleJitter :type Double) (Disturbance :type Double) (Disturbance_subdivisions_distance :type Double :default 4e+03) (Allow_Disturbance :type Bool))
  (outputs (SplittingPath :type Spline))
  ;[0.25 0.45 0.75 0.5] Sanitize input polygon
  ; Rebuild from spline control points and remove
  ; co-located vertices so downstream geometry is well-formed.
  (bind GetSplineControlPoints_1 (Get%20Spline%20Control%20Points))
  (bind CreateSpline_3 (Create%20Spline :InterpTypeAttribute None))
  (bind CleanSpline_0 (Clean%20Spline :ColocationDistanceThreshold 0.0001))
  (bind GetSplineControlPoints_2 (Get%20Spline%20Control%20Points))
  (bind CreatePolygon2D_0 (Create%20Polygon%202D))
  ;[0.25 0.45 0.75 0.5] Pick the longest outer edge
  ; Keep only the segment matching $LongestOuterSegmentIndex;
  ; the cut will run across this edge.
  (bind CopyAttributes_0 (Copy%20Attributes :InputSource @Data.$LongestOuterSegmentIndex :OutputTarget longestSeg))
  (bind CopyAttributes_1 (Copy%20Attributes :InputSource $SegmentLength :OutputTarget lenght))
  (bind CopyAttributes_2 (Copy%20Attributes :InputSource $LocalRotation :OutputTarget localRot))
  (bind AttributeCast_0 (Attribute%20Cast :OutputType Rotator))
  (bind GetSplineControlPoints_0 (Get%20Spline%20Control%20Points))
  (bind AttributeFilter_1 (Filter%20Attribute%20Elements :Operator Equal :TargetAttribute $Index :ThresholdAttribute longestSeg))
  ;[0.25 0.45 0.75 0.5] Reduce to a single anchor transform
  ; Combine that segment's control points into one anchor
  ; carrying its position and rotation.
  (bind CopyAttributes_5 (Copy%20Attributes :InputSource $Transform :OutputTarget PointTransform))
  (bind PointToAttributeSet_1 (Point%20To%20Attribute%20Set))
  (bind DeleteAttributes_2 (Delete%20Attributes :Operation KeepSelectedAttributes :SelectedAttributes "PointTransform"))
  (bind CombinePoints_0 (Combine%20Points :bUseFirstPointTransform false))
  ;[0.25 0.45 0.75 0.5] Random yaw (Angle Jitter)
  ; Build a random Yaw within ±AngleJitter so the cut
  ; is not always perfectly perpendicular.
  (bind AddAttribute_0 (Add%20Attribute :OutputTarget rand))
  (bind AttributeMathsOp_1 (Attribute%20Maths%20Op :Operation Multiply :InA AngleJitter))
  (bind AttributeNoise_0 (Attribute%20Noise :InputSource @Last :NoiseMax-param AngleJitter))
  (bind ExtractAttribute_1 (Extract%20Attribute%20at%20Index :bForceOutputAttributeToBeInElementsDomain true))
  (bind AttributeMathsOp_4 (Attribute%20Maths%20Op))
  (bind CreateConstant_2 (Create%20Constant :AttributeTypes (struct :Type Float)))
  (bind MakeRotatorAttribute_0 (Make%20Rotator%20Attribute :Operation MakeRotFromAngles))
  ;[0.25 0.45 0.75 0.5] Build the straight cut spline
  ; Rotate +90° plus jitter, swap extents to span the polygon,
  ; bias along it by Ratio, emit a two-point line.
  (bind TransformPoints_1 (Transform%20Points :RotationMax (struct :Yaw 9e+01) :RotationMin (struct :Yaw 9e+01)))
  (bind CopyAttributes_3 (Copy%20Attributes :InputSource $Extents :OutputTarget Extents))
  (bind AttributeMathsOp_2 (Attribute%20Maths%20Op :InputSource1 $Extents.x :InputSource2 Extents.Y :Operation Set))
  (bind AttributeMathsOp_3 (Attribute%20Maths%20Op :InputSource1 $Extents.y :InputSource2 Extents.X :Operation Set))
  (bind MakeVectorAttribute_1 (Make%20Vector%20Attribute :OutputType Vector :Y Ratio))
  (bind ResetPointCenter_2 (Reset%20Point%20Center :PointCenterLocation (0.5 0.817073 0.5)))
  (bind AttributeMathsOp_5 (Attribute%20Maths%20Op :InputSource1 $Extents.X :Operation Multiply))
  (bind CreatePoints_0 (Create%20Points))
  (bind AttributeMathsOp_0 (Attribute%20Maths%20Op :InputSource1 $Extents.x :InputSource2 $Extents.x :Operation Set))
  (bind ResetPointCenter_0 (Reset%20Point%20Center :PointCenterLocation (0.0 0.5 0.5)))
  (bind ResetPointCenter_1 (Reset%20Point%20Center :PointCenterLocation (1.0 0.5 0.5)))
  (bind MergePoints_0 (Merge%20Points))
  (bind CopyPoints_0 (Copy%20Points))
  (bind CreateSpline_0 (Create%20Spline :InterpTypeAttribute None))
  (bind Branch_0 (Branch :bOutputToB-param Allow_Disturbance)) ; Allow Disturbance routes to the wavy sampler;\notherwise the straight cut passes through.
  ;[0.25 0.45 0.75 0.5] Optional disturbance (wavy cut)
  ; Sample the straight spline, jitter interior points by
  ; the Disturbance vector; endpoints stay anchored.
  (bind SplineSampler_0 (Spline%20Sampler :SamplerParams (struct :Mode Distance :bUnbounded true :PointSteepness 1.0) :DistanceIncrement Disturbance_subdivisions_distance))
  (bind AttributeFilter_4 (Filter%20Attribute%20Elements :AttributeTypes (struct :Type Integer32) :bUseConstantThreshold true :Operator Equal :TargetAttribute $Index :ThresholdAttribute @Last))
  (bind GetElementsCount_2 (Get%20Element%20Count))
  (bind AttributeMathsOp_10 (Attribute%20Maths%20Op :Operation Subtract))
  (bind AttributeFilter_5 (Filter%20Attribute%20Elements :Operator Equal :TargetAttribute $Index :ThresholdAttribute @Last))
  (bind AttributeFilterRange_2 (Filter%20Attribute%20Elements%20by%20Range :MaxThreshold (struct :bInclusive false :ThresholdAttribute @Last :bUseSpatialQuery false) :MinThreshold (struct :bInclusive false :bUseConstantThreshold true :ThresholdAttribute @Last :bUseSpatialQuery false :AttributeTypes (struct :Type Integer32)) :TargetAttribute $Index))
  (bind MakeVectorAttribute_2 (Make%20Vector%20Attribute :OutputType Vector :X Disturbance :Y Disturbance))
  (bind AttributeMathsOp_9 (Attribute%20Maths%20Op :Operation Multiply))
  (bind TransformPoints_3 (Transform%20Points))
  (bind MergePoints_3 (Merge%20Points))
  (bind CreateSpline_1 (Create%20Spline :InterpTypeAttribute None))
  ;[0.25 0.45 0.75 0.5] Cut polygon & emit halves
  ; Slice the input polygon with the chosen path and clip
  ; the path to the polygon interior. Emit both halves
  ; plus the trimmed spline.
  (bind Polygon2DOperation_0 (Polygon%20Operation :Operation CutWithPaths))
  (bind ClipPaths_0 (Clip%20Paths))
  (bind FilterDataByType_1 (Filter%20Data%20By%20Type :TargetType (struct :Ids [(struct :Struct /Script/PCG.PCGDataTypeInfoSpline)])))
  (reroute poly%20input In)
  (connect CombinePoints_0 Out AddAttribute_0 In)
  (connect CopyAttributes_2 Out AttributeCast_0 In)
  (connect GetSplineControlPoints_0 Out AttributeFilter_1 In)
  (connect SplineSampler_0 Out AttributeFilter_4 In)
  (connect AttributeMathsOp_10 Out AttributeFilter_5 Filter)
  (connect SplineSampler_0 Out AttributeFilter_5 In)
  (connect AttributeMathsOp_10 Out AttributeFilterRange_2 FilterMax)
  (connect SplineSampler_0 Out AttributeFilterRange_2 In)
  (connect CreatePoints_0 Out AttributeMathsOp_0 InA)
  (connect AttributeMathsOp_5 Out AttributeMathsOp_0 InB)
  (connect CopyAttributes_3 Out AttributeMathsOp_2 InA)
  (connect CopyAttributes_3 Out AttributeMathsOp_2 InB)
  (connect AttributeMathsOp_2 Out AttributeMathsOp_3 InA)
  (connect AttributeMathsOp_2 Out AttributeMathsOp_3 InB)
  (connect ExtractAttribute_1 Out AttributeMathsOp_4 InA)
  (connect ResetPointCenter_2 Out AttributeMathsOp_5 InA)
  (connect MakeVectorAttribute_2 Out AttributeMathsOp_9 InA)
  (connect GetElementsCount_2 Out AttributeMathsOp_10 InA)
  (connect AddAttribute_0 Out AttributeNoise_0 In)
  (connect AttributeMathsOp_1 Out AttributeNoise_0 NoiseMin)
  (connect CreateSpline_0 Out Branch_0 In)
  (connect CreateSpline_3 Out CleanSpline_0 In)
  (connect poly%20input Out ClipPaths_0 ClipPolygons)
  (connect Branch_0 Output%20A ClipPaths_0 In)
  (connect CreateSpline_1 Out ClipPaths_0 In)
  (connect GetSplineControlPoints_0 Out CombinePoints_0 In)
  (connect DeleteAttributes_2 Out CombinePoints_0 PointTransform)
  (connect CreatePolygon2D_0 Out CopyAttributes_0 Source)
  (connect CreatePolygon2D_0 Out CopyAttributes_0 Target)
  (connect CopyAttributes_0 Out CopyAttributes_1 Source)
  (connect CopyAttributes_0 Out CopyAttributes_1 Target)
  (connect CopyAttributes_1 Out CopyAttributes_2 Source)
  (connect CopyAttributes_1 Out CopyAttributes_2 Target)
  (connect TransformPoints_1 Out CopyAttributes_3 Source)
  (connect TransformPoints_1 Out CopyAttributes_3 Target)
  (connect AttributeFilter_1 InsideFilter CopyAttributes_5 Source)
  (connect AttributeFilter_1 InsideFilter CopyAttributes_5 Target)
  (connect MergePoints_0 Out CopyPoints_0 Source)
  (connect ResetPointCenter_2 Out CopyPoints_0 Target)
  (connect GetSplineControlPoints_2 Out CreatePolygon2D_0 In)
  (connect CopyPoints_0 Out CreateSpline_0 In)
  (connect MergePoints_3 Out CreateSpline_1 In)
  (connect GetSplineControlPoints_1 Out CreateSpline_3 In)
  (connect PointToAttributeSet_1 Out DeleteAttributes_2 In)
  (connect AttributeNoise_0 Out ExtractAttribute_1 In)
  (connect ClipPaths_0 Out FilterDataByType_1 In)
  (connect SplineSampler_0 Out GetElementsCount_2 In)
  (connect AttributeCast_0 Out GetSplineControlPoints_0 In)
  (connect poly%20input Out GetSplineControlPoints_1 In)
  (connect CleanSpline_0 Out GetSplineControlPoints_2 In)
  (connect CreateConstant_2 Out MakeRotatorAttribute_0 Pitch)
  (connect CreateConstant_2 Out MakeRotatorAttribute_0 Roll)
  (connect AttributeMathsOp_4 Out MakeRotatorAttribute_0 Yaw)
  (connect ResetPointCenter_0 Out MergePoints_0 In)
  (connect ResetPointCenter_1 Out MergePoints_0 In)
  (connect AttributeFilter_4 InsideFilter MergePoints_3 In)
  (connect AttributeFilter_5 InsideFilter MergePoints_3 In)
  (connect TransformPoints_3 Out MergePoints_3 In)
  (connect CopyAttributes_5 Out PointToAttributeSet_1 In)
  (connect Branch_0 Output%20A Polygon2DOperation_0 ClipPaths)
  (connect CreateSpline_1 Out Polygon2DOperation_0 ClipPaths)
  (connect poly%20input Out Polygon2DOperation_0 In)
  (connect AttributeMathsOp_0 Out ResetPointCenter_0 In)
  (connect AttributeMathsOp_0 Out ResetPointCenter_1 In)
  (connect AttributeMathsOp_3 Out ResetPointCenter_2 In)
  (connect MakeVectorAttribute_1 Out ResetPointCenter_2 PointCenterLocation)
  (connect Branch_0 Output%20B SplineSampler_0 Spline)
  (connect CombinePoints_0 Out TransformPoints_1 In)
  (connect MakeRotatorAttribute_0 Out TransformPoints_1 RotationMax)
  (connect MakeRotatorAttribute_0 Out TransformPoints_1 RotationMin)
  (connect AttributeFilterRange_2 InsideFilter TransformPoints_3 In)
  (connect MakeVectorAttribute_2 Out TransformPoints_3 OffsetMax)
  (connect AttributeMathsOp_9 Out TransformPoints_3 OffsetMin)
  (output Out Polygon2DOperation_0)
  (output SplittingPath FilterDataByType_1.InsideFilter))
