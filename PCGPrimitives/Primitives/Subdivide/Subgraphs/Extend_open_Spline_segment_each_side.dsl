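; PCG subgraph that extends an open spline segment at both ends by a distance parameter, defaulting to 50.
; It splits off the last control point from the rest, then subtracts positions and normalizes to get the segment direction.
; The scaled direction offsets the last point, and a further multiplied copy offsets the other side.
; Both results merge into a new linear spline, which suggests two-point segments as the intended input.
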(graph Extend_open_Spline_segment_each_side
  (params (distance :type Double :default 5e+01))
  (bind GetSplineControlPoints_0 (Get%20Spline%20Control%20Points))
  (bind AttributeFilter_0 (Filter%20Attribute%20Elements :bUseConstantThreshold true :TargetAttribute $Index :ThresholdAttribute @Last))
  (bind AttributeMathsOp_0 (Attribute%20Maths%20Op :InputSource1 $Position :InputSource2 $Position :Operation Subtract :OutputTarget normal))
  (bind AttributeVectorOp_0 (Attribute%20Vector%20Op :InputSource1 normal :Operation Normalize))
  (bind AttributeMathsOp_1 (Attribute%20Maths%20Op :InputSource1 normal :Operation Multiply :InB distance))
  (bind AttributeMathsOp_2 (Attribute%20Maths%20Op :InputSource1 $Position :InputSource2 normal))
  (bind AttributeMathsOp_3 (Attribute%20Maths%20Op :InputSource1 normal :Operation Multiply))
  (bind AttributeMathsOp_4 (Attribute%20Maths%20Op :InputSource1 $Position :InputSource2 normal))
  (bind MergePoints_0 (Merge%20Points))
  (bind CreateSpline_0 (Create%20Spline :bLinear true :InterpTypeAttribute None))
  (connect GetSplineControlPoints_0 Out AttributeFilter_0 In)
  (connect AttributeFilter_0 InsideFilter AttributeMathsOp_0 InA)
  (connect AttributeFilter_0 OutsideFilter AttributeMathsOp_0 InB)
  (connect AttributeVectorOp_0 Out AttributeMathsOp_1 InA)
  (connect AttributeMathsOp_1 Out AttributeMathsOp_2 InA)
  (connect AttributeMathsOp_1 Out AttributeMathsOp_2 InB)
  (connect AttributeMathsOp_1 Out AttributeMathsOp_3 InA)
  (connect AttributeFilter_0 OutsideFilter AttributeMathsOp_4 InA)
  (connect AttributeMathsOp_3 Out AttributeMathsOp_4 InB)
  (connect AttributeMathsOp_0 Out AttributeVectorOp_0 In)
  (connect MergePoints_0 Out CreateSpline_0 In)
  (connect DefaultInputNode In GetSplineControlPoints_0 In)
  (connect AttributeMathsOp_2 Out MergePoints_0 In)
  (connect AttributeMathsOp_4 Out MergePoints_0 In)
  (output Out CreateSpline_0))
