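; PCG loop-body subgraph for the grid-lines subdivide primitive that re-orients a line spline by a random yaw controlled by angleJitter.
; Attribute Noise, seeded by loop index, produces the yaw for a rotator with constant pitch and roll.
; The spline's control points are combined into one point and rotated by that rotator.
; Resetting the point center to both X ends yields two endpoints, merged into a new output spline.
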(graph Grid_lines_orient_loop
  (params (angleJitter :type Double))
  (bind AddAttribute_1 (Add%20Attribute :OutputTarget rand))
  (bind AttributeMathsOp_0 (Attribute%20Maths%20Op :Operation Multiply :InA AngleJitter))
  (bind GetLoopIndex_0 (Get%20Loop%20Index))
  (bind AttributeNoise_1 (Attribute%20Noise :InputSource @Last :NoiseMax-param AngleJitter))
  (bind ExtractAttribute_0 (Extract%20Attribute%20at%20Index :bForceOutputAttributeToBeInElementsDomain true))
  (bind AttributeMathsOp_1 (Attribute%20Maths%20Op))
  (bind GetSplineControlPoints_1 (Get%20Spline%20Control%20Points))
  (bind CombinePoints_0 (Combine%20Points))
  (bind CreateConstant_0 (Create%20Constant :AttributeTypes (struct :Type Float)))
  (bind MakeRotatorAttribute_1 (Make%20Rotator%20Attribute :Operation MakeRotFromAngles))
  (bind TransformPoints_0 (Transform%20Points :RotationMax (struct :Yaw 9e+01) :RotationMin (struct :Yaw 9e+01)))
  (bind GetSplineControlPoints_0 (Get%20Spline%20Control%20Points))
  (bind ResetPointCenter_0 (Reset%20Point%20Center :PointCenterLocation (0.0 0.5 0.5)))
  (bind ResetPointCenter_1 (Reset%20Point%20Center :PointCenterLocation (1.0 0.5 0.5)))
  (bind MergePoints_0 (Merge%20Points))
  (bind CreateSpline_0 (Create%20Spline :InterpTypeAttribute None))
  (connect DefaultInputNode In AddAttribute_1 In)
  (connect ExtractAttribute_0 Out AttributeMathsOp_1 InA)
  (connect AddAttribute_1 Out AttributeNoise_1 In)
  (connect AttributeMathsOp_0 Out AttributeNoise_1 NoiseMin)
  (connect GetLoopIndex_0 Out AttributeNoise_1 Seed)
  (connect GetSplineControlPoints_1 Out CombinePoints_0 In)
  (connect MergePoints_0 Out CreateSpline_0 In)
  (connect AttributeNoise_1 Out ExtractAttribute_0 In)
  (connect DefaultInputNode In GetSplineControlPoints_0 In)
  (connect DefaultInputNode In GetSplineControlPoints_1 In)
  (connect CreateConstant_0 Out MakeRotatorAttribute_1 Pitch)
  (connect CreateConstant_0 Out MakeRotatorAttribute_1 Roll)
  (connect ExtractAttribute_0 Out MakeRotatorAttribute_1 Yaw)
  (connect ResetPointCenter_0 Out MergePoints_0 In)
  (connect ResetPointCenter_1 Out MergePoints_0 In)
  (connect TransformPoints_0 Out ResetPointCenter_0 In)
  (connect TransformPoints_0 Out ResetPointCenter_1 In)
  (connect CombinePoints_0 Out TransformPoints_0 In)
  (connect MakeRotatorAttribute_1 Out TransformPoints_0 RotationMax)
  (connect MakeRotatorAttribute_1 Out TransformPoints_0 RotationMin)
  (output Out CreateSpline_0))
