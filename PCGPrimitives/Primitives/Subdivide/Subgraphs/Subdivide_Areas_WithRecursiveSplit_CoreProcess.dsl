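; PCG subgraph that subdivides Polygon2D areas by recursive splitting, exposing width, iteration, minimum size, seed, angle jitter and disturbance controls.
; A Loop runs the RecursiveSplit_Loop subgraph on each polygon, yielding split areas and splitting paths.
; Areas become outer-ring-only closed linear splines tagged Type:ClosedSpline, while paths are retagged Type:Spline.
; Both sets have rotation forced to zero and share one output.
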(graph Subdivide_Areas_WithRecursiveSplit_CoreProcess
  (params (SplitingWidth :type Double :default 1e+02) (MaxIterations :type Int :default 4) (MaxSplitDistanceFromCenter :type Double :default 0.3) (MinAreaSize :type Double :default 1e+03) (Seed :type Int) (AngleJitter :type Double) (Disturbance :type Double) (Disturbance_Subdivisions :type Double :default 4e+03) (Allow_Disturbance :type Bool))
  (bind FilterDataByType_2 (Filter%20Data%20By%20Type :TargetType (struct :Ids [(struct :Struct /Script/PCG.PCGDataTypeInfoPolygon2D)])))
  (bind Loop_1 (Loop :Graph /PCGPrimitives/Primitives/Subdivide/Subgraphs/Subdivide_Areas_WithRecursiveSplit_Loop.Subdivide_Areas_WithRecursiveSplit_Loop :Allow_Disturbance Allow_Disturbance :AngleJitter AngleJitter :Disturbance Disturbance :Disturbance_Subdivision Disturbance_Subdivisions :MaxIterations MaxIterations :MaxSplitDistanceFromCenter MaxSplitDistanceFromCenter :MinAreaSize MinAreaSize :Seed-param Seed :SplitingWidth SplitingWidth))
  (bind ToPoint_0 (To%20Point))
  (bind FilterDataByAttribute_0 (Filter%20Data%20By%20Attribute :FilterMode FilterByValue :FilterOperator Equal :TargetAttribute HoleIndex :Threshold (struct :AttributeTypes (struct :Type Integer32 :Int32Value -1)))) ; Drop hole interiors — keep only outer rings.
  (bind CreateSpline_0 (Create%20Spline :bClosedLoop true :bLinear true :InterpTypeAttribute None))
  (bind DeleteTags_0 (Delete%20Tags :SelectedTags "Type:Point, Type:Spline, Type:ClosedSpline, Type:Polygon2D, Type:Surface, Type:Texture2D, Type:Volume"))
  (bind AddTags_0 (Add%20Tags :TagsToAdd "Type:Spline"))
  (bind DeleteTags_1 (Delete%20Tags :SelectedTags "Type:Point, Type:Spline, Type:ClosedSpline, Type:Polygon2D, Type:Surface, Type:Texture2D, Type:Volume"))
  (bind AddTags_1 (Add%20Tags :TagsToAdd "Type:ClosedSpline"))
  (bind CreateConstant_1 (Create%20Constant))
  (bind MakeRotatorAttribute_1 (Make%20Rotator%20Attribute :Operation MakeRotFromAngles))
  ;[0.15 0.15 0.15 0.5] Force Rotation 0
  (bind AttributeMathsOp_1 (Attribute%20Maths%20Op :InputSource1 $Rotation :Operation Set))
  (bind AttributeMathsOp_2 (Attribute%20Maths%20Op :InputSource1 $Rotation :Operation Set))
  (connect DeleteTags_0 Out AddTags_0 In)
  (connect DeleteTags_1 Out AddTags_1 In)
  (connect AddTags_1 Out AttributeMathsOp_1 InA)
  (connect MakeRotatorAttribute_1 Out AttributeMathsOp_1 InB)
  (connect AddTags_0 Out AttributeMathsOp_2 InA)
  (connect MakeRotatorAttribute_1 Out AttributeMathsOp_2 InB)
  (connect FilterDataByAttribute_0 InsideFilter CreateSpline_0 In)
  (connect Loop_1 SplittingPath DeleteTags_0 In)
  (connect CreateSpline_0 Out DeleteTags_1 In)
  (connect ToPoint_0 Out FilterDataByAttribute_0 In)
  (connect DefaultInputNode In FilterDataByType_2 In)
  (connect FilterDataByType_2 InsideFilter Loop_1 In)
  (connect CreateConstant_1 Out MakeRotatorAttribute_1 Pitch)
  (connect CreateConstant_1 Out MakeRotatorAttribute_1 Roll)
  (connect CreateConstant_1 Out MakeRotatorAttribute_1 Yaw)
  (connect Loop_1 Out ToPoint_0 In)
  (output Out AttributeMathsOp_1)
  (output Out AttributeMathsOp_2))
