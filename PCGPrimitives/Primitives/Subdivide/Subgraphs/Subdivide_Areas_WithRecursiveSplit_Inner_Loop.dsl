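; PCG subgraph for the inner loop of recursive area subdivision, splitting a polygon at a randomized position along its longest edge.
; It derives a per-polygon seed from the spline control points and Seed parameter, then generates noise.
; That noise is remapped into a symmetric range set by MaxSplitDistanceFromCenter and feeds AutoSplitPolygon as the split ratio.
; AutoSplitPolygon also receives width, angle jitter and disturbance settings, returning split results and the SplittingPath spline.
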(graph Subdivide_Areas_WithRecursiveSplit_Inner_Loop
  (params (MaxSplitDistanceFromCenter :type Double :default 0.1) (Seed :type Int) (SplitingWidth :type Double :default 1e+02) (angleJitter :type Double) (Disturbance :type Double) (Disturbance_subdivisions :type Double :default 4e+03) (Allow_Disturbance :type Bool))
  (outputs (SplittingPath :type Spline))
  (bind Subgraph_0 (Subgraph :Graph /PCGPrimitives/Primitives/Subdivide/Subgraphs/AutoSplitPolygon.AutoSplitPolygon :Allow_Disturbance Allow_Disturbance :AngleJitter angleJitter :Disturbance Disturbance :Disturbance_subdivisions_distance Disturbance_subdivisions :SplitingWidth SplitingWidth))
  ;[0.25 0.45 0.75 0.5] Derive a per-polygon seed from the input control
  ; points and the user seed, then roll a noise value
  ; used to randomize where along the edge the split occurs.
  (bind GetSplineControlPoints_0 (Get%20Spline%20Control%20Points))
  (bind MutateSeed_2 (Mutate%20Seed))
  (bind AttributeMathsOp_3 (Attribute%20Maths%20Op :InputSource1 $Seed :InB Seed))
  (bind AttributeNoise_1 (Attribute%20Noise :OutputTarget attr))
  (bind GetAttributeFromPointIndex_2 (Get%20Attribute%20From%20Point%20Index :bForceOutputAttributeToBeInElementsDomain true))
  ;[0.25 0.45 0.75 0.5] Compute the allowed range for the split ratio,
  ; symmetric around the edge center: [d, 1-d] where
  ; d is the clamped Max Split Distance from Center.
  (bind AttributeMathsOp_2 (Attribute%20Maths%20Op :Operation Clamp :In MaxSplitDistanceFromCenter))
  (bind AttributeMathsOp_0 (Attribute%20Maths%20Op))
  (bind AttributeMathsOp_1 (Attribute%20Maths%20Op :Operation OneMinus))
  ;[0.25 0.45 0.75 0.5] Remap the noise into the computed range to get
  ; the final split ratio, then cut the polygon along
  ; its longest edge at that position via AutoSplitPolygon.
  (bind AttributeRemap_0 (Attribute%20Remap))
  (connect AttributeMathsOp_2 Out AttributeMathsOp_0 InA)
  (connect AttributeMathsOp_0 Out AttributeMathsOp_1 In)
  (connect MutateSeed_2 Out AttributeMathsOp_3 InA)
  (connect AttributeMathsOp_3 Out AttributeNoise_1 In)
  (connect GetAttributeFromPointIndex_2 Attribute AttributeRemap_0 In)
  (connect AttributeMathsOp_1 Out AttributeRemap_0 OutRangeMax)
  (connect AttributeMathsOp_0 Out AttributeRemap_0 OutRangeMin)
  (connect AttributeNoise_1 Out GetAttributeFromPointIndex_2 In)
  (connect DefaultInputNode In GetSplineControlPoints_0 In)
  (connect GetSplineControlPoints_0 Out MutateSeed_2 In)
  (connect DefaultInputNode In Subgraph_0 In)
  (connect AttributeRemap_0 Out Subgraph_0 Ratio)
  (output Out Subgraph_0)
  (output SplittingPath Subgraph_0.SplittingPath))
