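; Recursive PCG subgraph that subdivides polygon areas and collects the splitting paths as splines.
; Areas below MinAreaSize, given in square metres, pass straight out, as do all areas once CurrentIteration reaches MaxIterations.
; Remaining areas are split by the Inner_Loop subgraph, using width, center-offset, angle jitter and optional disturbance settings.
; The graph loops itself over the resulting pieces with an incremented iteration, gathering splines from every level.
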(graph Subdivide_Areas_WithRecursiveSplit_Loop
  (params (SplitingWidth :type Double :default 1e+02) (CurrentIteration :type Int) (MaxIterations :type Int :default 2) (MaxSplitDistanceFromCenter :type Double :default 0.1) (MinAreaSize :type Double :default 1e+03) (Seed :type Int) (AngleJitter :type Double) (Disturbance :type Double) (Disturbance_Subdivision :type Double :default 4e+03) (Allow_Disturbance :type Bool))
  (outputs (SplittingPath :type Spline))
  ;[0.25 0.45 0.75 0.5] Stop once CurrentIteration reaches Max; otherwise
  ; split once via Inner Loop and feed both halves back
  ; into this graph with iteration+1, gathering splines.
  (bind AttributeMathsOp_1 (Attribute%20Maths%20Op :InA CurrentIteration))
  (bind AttributeCompareOp_1 (Attribute%20Compare%20Op :Operation Less :InA CurrentIteration :InB MaxIterations))
  (bind Branch_0 (Branch))
  (bind Loop_1 (Loop :Graph /PCGPrimitives/Primitives/Subdivide/Subgraphs/Subdivide_Areas_WithRecursiveSplit_Inner_Loop.Subdivide_Areas_WithRecursiveSplit_Inner_Loop :Allow_Disturbance Allow_Disturbance :angleJitter AngleJitter :Disturbance Disturbance :Disturbance_subdivisions Disturbance_Subdivision :MaxSplitDistanceFromCenter MaxSplitDistanceFromCenter :Seed-param Seed :SplitingWidth SplitingWidth))
  (bind Loop_2 (Loop :Graph /PCGPrimitives/Primitives/Subdivide/Subgraphs/Subdivide_Areas_WithRecursiveSplit_Loop.Subdivide_Areas_WithRecursiveSplit_Loop :Allow_Disturbance Allow_Disturbance :AngleJitter AngleJitter :Disturbance Disturbance :Disturbance_Subdivision Disturbance_Subdivision :MaxIterations MaxIterations :MaxSplitDistanceFromCenter MaxSplitDistanceFromCenter :MinAreaSize MinAreaSize :Seed-param Seed :SplitingWidth SplitingWidth))
  (bind Gather_0 (Gather))
  ;[0.25 0.45 0.75 0.5] Skip polygons smaller than MinAreaSize (converted
  ; from m² to cm²); they pass straight to the output
  ; without further subdivision.
  (bind AttributeMathsOp_0 (Attribute%20Maths%20Op :Operation Multiply :InA MinAreaSize)) ; Making the input being square meter instead of square cm 
  (bind FilterDataByAttribute_1 (Filter%20Data%20By%20Attribute :FilterMode FilterByValue :TargetAttribute @Data.$Area :Threshold (struct :bUseConstantThreshold false :AttributeTypes (struct :DoubleValue 5e+03))))
  (connect AttributeCompareOp_1 Out Branch_0 bOutputToB)
  (connect FilterDataByAttribute_1 InsideFilter Branch_0 In)
  (connect AttributeMathsOp_0 Out FilterDataByAttribute_1 Filter)
  (connect DefaultInputNode In FilterDataByAttribute_1 In)
  (connect Loop_2 SplittingPath Gather_0 In)
  (connect Loop_1 SplittingPath Gather_0 In2)
  (connect Branch_0 Output%20B Loop_1 In)
  (connect AttributeMathsOp_1 Out Loop_2 CurrentIteration)
  (connect Loop_1 Out Loop_2 In)
  (output Out Branch_0.Output%20A)
  (output Out FilterDataByAttribute_1.OutsideFilter)
  (output Out Loop_2)
  (output SplittingPath Gather_0))
