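; PCG subgraph that subdivides a polygon into closed cell splines and clipped divider splines, using a grid aligned to its longest outer edge.
; Cell size comes from fixed dimensions or subdivision counts; grid points partition by X/Y into lines, which loops extend and angle-jitter.
; Optional disturbance resamples lines and randomly offsets middle points for curved divisions.
; Lines cut the polygon in grid-local space, then results transform back, tagged by spline type.
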(graph Subdivide_Areas_withGridLoop
  (params (GridWidth :type Double :default 5e+03) (GridLength :type Double :default 5e+03) (Subdiv_width :type Int :default 1) (Subdiv_length :type Int :default 1) (Use_Subdivision_instead_of_sizes :type Bool) (AngleJitter :type Double) (Allow_Disturbance :type Bool) (Disturbance :type Double) (Disturbance_Subdivisions_distance :type Double :default 4e+03))
  (bind CreatePolygon2D_0 (Create%20Polygon%202D :InputType ForceClosed))
  (bind GetAttributeFromPointIndex_6 (Get%20Attribute%20From%20Point%20Index :bForceOutputAttributeToBeInElementsDomain true :InputSource $Transform))
  (bind CreateConstant_1 (Create%20Constant))
  (bind MakeRotatorAttribute_1 (Make%20Rotator%20Attribute :Operation MakeRotFromAngles))
  ;[0.25 0.45 0.75 0.5] Extract longest outer edge of the polygon and build
  ; a PointTransform that orients the grid along it.
  (bind CopyAttributes_1 (Copy%20Attributes :InputSource @Data.$LongestOuterSegmentIndex :OutputTarget longestSeg))
  (bind CopyAttributes_2 (Copy%20Attributes :InputSource $SegmentLength :OutputTarget lenght))
  (bind CopyAttributes_3 (Copy%20Attributes :InputSource $LocalRotation :OutputTarget localRot))
  (bind AttributeCast_1 (Attribute%20Cast :OutputType Rotator))
  (bind ToPoint_0 (To%20Point))
  (bind AttributeFilter_2 (Filter%20Attribute%20Elements :Operator Equal :TargetAttribute $Index :ThresholdAttribute longestSeg))
  (bind CopyAttributes_4 (Copy%20Attributes :InputSource $Transform :OutputTarget PointTransform))
  (bind PointToAttributeSet_0 (Point%20To%20Attribute%20Set))
  (bind DeleteAttributes_0 (Delete%20Attributes :Operation KeepSelectedAttributes :SelectedAttributes "PointTransform"))
  (bind CombinePoints_2 (Combine%20Points :bUseFirstPointTransform false))
  ;[0.25 0.45 0.75 0.5] Use_Subdivision_instead_of_sizes :
  ; Pick cell size: absolute centimeters from Grid Width/Length,
  ; or derive it by dividing extents by Subdiv counts.
  (bind GetAttributeFromPointIndex_1 (Get%20Attribute%20From%20Point%20Index :bForceOutputAttributeToBeInElementsDomain true :InputSource $Extents))
  (bind AttributeMathsOp_7 (Attribute%20Maths%20Op :Operation Multiply))
  (bind MakeVectorAttribute_2 (Make%20Vector%20Attribute :OutputType Vector :X Subdiv_width :Y Subdiv_length))
  (bind AttributeMathsOp_3 (Attribute%20Maths%20Op :Operation Divide))
  (bind MakeVectorAttribute_0 (Make%20Vector%20Attribute :OutputType Vector :X GridWidth :Y GridLength))
  (bind Select_0 (Select :bUseInputB-param Use_Subdivision_instead_of_sizes))
  (bind MakeVectorAttribute_1 (Make%20Vector%20Attribute :OutputType Vector :X GridWidth :Y GridLength))
  (bind AttributeMathsOp_0 (Attribute%20Maths%20Op))
  (bind AttributeMathsOp_6 (Attribute%20Maths%20Op :Operation Multiply))
  (bind AttributeMathsOp_5 (Attribute%20Maths%20Op))
  (bind Select_1 (Select :bUseInputB-param Use_Subdivision_instead_of_sizes))
  ;[0.25 0.45 0.75 0.5] Generate the oriented point grid, then partition points
  ; by X and by Y to form perpendicular grid lines.
  (bind CreatePointsGrid_1 (Create%20Points%20Grid :CellSize (5e+03 5e+03 1e+04) :PointPosition CellCorners))
  (bind AttributeNoise_0 (Attribute%20Noise :InputSource $Density))
  (bind AttributePartition_1 (Attribute%20Partition :PartitionAttributeSelectors [$Position.x]))
  (bind AttributePartition_3 (Attribute%20Partition :PartitionAttributeSelectors [$Position.y]))
  (bind AddTags_2 (Add%20Tags :TagsToAdd "X"))
  (bind AddTags_3 (Add%20Tags :TagsToAdd "Y"))
  ;;[0.15 0.15 0.15 0.5] REMOVING first "ROAD" on the border
  (bind FilterDataByIndex_0 (Filter%20Data%20By%20Index :bInvertFilter true :SelectedIndices "0,1"))
  (bind AttributeFilter_0 (Filter%20Attribute%20Elements :AttributeTypes (struct :Type Integer32) :bUseConstantThreshold true :Operator Equal :TargetAttribute $Index :ThresholdAttribute @Last))
  (bind FilterDataByIndex_1 (Filter%20Data%20By%20Index :bInvertFilter true :SelectedIndices "0,1"))
  (bind AttributeFilter_1 (Filter%20Attribute%20Elements :AttributeTypes (struct :Type Integer32) :bUseConstantThreshold true :Operator Equal :TargetAttribute $Index :ThresholdAttribute @Last))
  ;[0.25 0.45 0.75 0.5] Cut the polygon by the lines to produce closed
  ; cells; clip the lines to the polygon outline; tag
  ; outputs as ClosedSpline / Spline and zero their rotation.
  (bind AttributeTransformOp_1 (Attribute%20Transform%20Op :Operation Invert))
  (bind AttributeTransformOp_4 (Attribute%20Transform%20Op :InputSource1 @Data.$Transform :InputSource2 Transform))
  (bind Polygon2DOperation_2 (Polygon%20Operation :Operation CutWithPaths))
  (bind AttributeTransformOp_2 (Attribute%20Transform%20Op :InputSource1 @Data.$Transform :InputSource2 Transform))
  (bind ToPoint_7 (To%20Point))
  (bind CreateSpline_4 (Create%20Spline :bClosedLoop true :bLinear true :InterpTypeAttribute None))
  (bind AddTags_1 (Add%20Tags :TagsToAdd "Type:ClosedSpline")) ; ClosedSpline
  (bind AttributeMathsOp_1 (Attribute%20Maths%20Op :InputSource1 $Rotation :Operation Set))
  ;[0.25 0.45 0.75 0.5] Take first and last point of each row to
  ; form lines, extend past bounds, apply optional angle jitter.
  (bind Gather_0 (Gather))
  (bind FilterElementsByIndex_1 (Filter%20Elements%20By%20Index :bSelectIndicesByInput false :SelectedIndices "0, -1"))
  (bind CreateSpline_3 (Create%20Spline :InterpTypeAttribute None))
  (bind Loop_0 (Loop :Graph /PCGPrimitives/Primitives/Subdivide/Subgraphs/Extend_open_Spline_segment_each_side.Extend_open_Spline_segment_each_side))
  (bind Loop_1 (Loop :Graph /PCGPrimitives/Primitives/Subdivide/Subgraphs/Grid_lines_orient_loop.Grid_lines_orient_loop :angleJitter AngleJitter))
  ;[0.25 0.45 0.75 0.5] Optional: resample each line and offset its middle points
  ; to produce curved, organic subdivisions instead of straight lines. Disturbance.
  (bind Branch_1 (Branch :bOutputToB-param Allow_Disturbance))
  (bind WriteDataIndex_0 (Write%20Data%20Index :bAddAttribute true :bAddTag false :IndexAttribute dataIndex))
  (bind SplineSampler_1 (Spline%20Sampler :SamplerParams (struct :Mode Distance :bUnbounded true :PointSteepness 1.0) :DistanceIncrement Disturbance_subdivisions_distance))
  (bind CopyAttributes_0 (Copy%20Attributes :InputSource $Index :OutputTarget storedIndexes))
  (bind AttributeFilter_3 (Filter%20Attribute%20Elements :AttributeTypes (struct :Type Integer32) :bUseConstantThreshold true :Operator Equal :TargetAttribute $Index :ThresholdAttribute @Last))
  (bind GetElementsCount_0 (Get%20Element%20Count))
  (bind AttributeMathsOp_8 (Attribute%20Maths%20Op :Operation Subtract))
  (bind AttributeFilter_4 (Filter%20Attribute%20Elements :Operator Equal :TargetAttribute $Index :ThresholdAttribute @Last))
  (bind AttributeFilterRange_0 (Filter%20Attribute%20Elements%20by%20Range :MaxThreshold (struct :bInclusive false :ThresholdAttribute @Last :bUseSpatialQuery false) :MinThreshold (struct :bInclusive false :bUseConstantThreshold true :ThresholdAttribute @Last :bUseSpatialQuery false :AttributeTypes (struct :Type Integer32)) :TargetAttribute $Index))
  (bind MakeVectorAttribute_3 (Make%20Vector%20Attribute :OutputType Vector :X Disturbance :Y Disturbance))
  (bind AttributeMathsOp_4 (Attribute%20Maths%20Op :Operation Multiply))
  (bind TransformPoints_0 (Transform%20Points))
  (bind MergePoints_0 (Merge%20Points))
  (bind AttributePartition_0 (Attribute%20Partition :PartitionAttributeSelectors [dataIndex]))
  (bind SortAttributes_0 (Sort%20Attributes :bUseStableSort false :SortAttributes [(struct :InputSource storedIndexes)]))
  (bind CreateSpline_0 (Create%20Spline :InterpTypeAttribute None))
  ;[0.15 0.15 0.15 0.5] SUBDIVING Splines
  (bind ClipPaths_1 (Clip%20Paths))
  (bind AttributeTransformOp_5 (Attribute%20Transform%20Op :InputSource1 @Data.$splineTransform :InputSource2 Transform))
  (bind FilterDataByType_0 (Filter%20Data%20By%20Type :TargetType (struct :Ids [(struct :Struct /Script/PCG.PCGDataTypeInfoPolyline)])))
  (bind GetSplineControlPoints_2 (Get%20Spline%20Control%20Points))
  (bind Branch_2 (Branch :bOutputToB-param Allow_Disturbance))
  (bind CreateSpline_1 (Create%20Spline :InterpTypeAttribute None))
  (bind CreateSpline_2 (Create%20Spline :bLinear true :InterpTypeAttribute None))
  (bind AddTags_4 (Add%20Tags :TagsToAdd "Type:Spline")) ; Spline
  (bind AttributeMathsOp_2 (Attribute%20Maths%20Op :InputSource1 $Rotation :Operation Set))
  (reroute InputPolygon2D CreatePolygon2D_0)
  (connect CreateSpline_4 Out AddTags_1 In)
  (connect AttributePartition_1 Out AddTags_2 In)
  (connect AttributePartition_3 Out AddTags_3 In)
  (connect CreateSpline_1 Out AddTags_4 In)
  (connect CreateSpline_2 Out AddTags_4 In)
  (connect CopyAttributes_3 Out AttributeCast_1 In)
  (connect FilterDataByIndex_0 InsideFilter AttributeFilter_0 In)
  (connect FilterDataByIndex_1 InsideFilter AttributeFilter_1 In)
  (connect ToPoint_0 Out AttributeFilter_2 In)
  (connect CopyAttributes_0 Out AttributeFilter_3 In)
  (connect AttributeMathsOp_8 Out AttributeFilter_4 Filter)
  (connect CopyAttributes_0 Out AttributeFilter_4 In)
  (connect AttributeMathsOp_8 Out AttributeFilterRange_0 FilterMax)
  (connect CopyAttributes_0 Out AttributeFilterRange_0 In)
  (connect GetAttributeFromPointIndex_1 Attribute AttributeMathsOp_0 InA)
  (connect MakeVectorAttribute_1 Out AttributeMathsOp_0 InB)
  (connect AddTags_1 Out AttributeMathsOp_1 InA)
  (connect MakeRotatorAttribute_1 Out AttributeMathsOp_1 InB)
  (connect AddTags_4 Out AttributeMathsOp_2 InA)
  (connect MakeRotatorAttribute_1 Out AttributeMathsOp_2 InB)
  (connect AttributeMathsOp_7 Out AttributeMathsOp_3 InA)
  (connect MakeVectorAttribute_2 Out AttributeMathsOp_3 InB)
  (connect MakeVectorAttribute_3 Out AttributeMathsOp_4 InA)
  (connect GetAttributeFromPointIndex_1 Attribute AttributeMathsOp_5 InA)
  (connect AttributeMathsOp_6 Out AttributeMathsOp_5 InB)
  (connect AttributeMathsOp_3 Out AttributeMathsOp_6 InA)
  (connect GetAttributeFromPointIndex_1 Attribute AttributeMathsOp_7 InA)
  (connect GetElementsCount_0 Out AttributeMathsOp_8 InA)
  (connect CreatePointsGrid_1 Out AttributeNoise_0 In)
  (connect MergePoints_0 Out AttributePartition_0 In)
  (connect AttributeNoise_0 Out AttributePartition_1 In)
  (connect AttributeNoise_0 Out AttributePartition_3 In)
  (connect GetAttributeFromPointIndex_6 Attribute AttributeTransformOp_1 In)
  (connect Polygon2DOperation_2 Out AttributeTransformOp_2 InA)
  (connect GetAttributeFromPointIndex_6 Attribute AttributeTransformOp_2 InB)
  (connect InputPolygon2D Out AttributeTransformOp_4 InA)
  (connect AttributeTransformOp_1 Out AttributeTransformOp_4 InB)
  (connect ClipPaths_1 Out AttributeTransformOp_5 InA)
  (connect GetAttributeFromPointIndex_6 Attribute AttributeTransformOp_5 InB)
  (connect Loop_1 Out Branch_1 In)
  (connect GetSplineControlPoints_2 Out Branch_2 In)
  (connect AttributeTransformOp_4 Out ClipPaths_1 ClipPolygons)
  (connect Branch_1 Output%20A ClipPaths_1 In)
  (connect CreateSpline_0 Out ClipPaths_1 In)
  (connect ToPoint_0 Out CombinePoints_2 In)
  (connect DeleteAttributes_0 Out CombinePoints_2 PointTransform)
  (connect SplineSampler_1 Out CopyAttributes_0 Source)
  (connect SplineSampler_1 Out CopyAttributes_0 Target)
  (connect CreatePolygon2D_0 Out CopyAttributes_1 Source)
  (connect CreatePolygon2D_0 Out CopyAttributes_1 Target)
  (connect CopyAttributes_1 Out CopyAttributes_2 Source)
  (connect CopyAttributes_1 Out CopyAttributes_2 Target)
  (connect CopyAttributes_2 Out CopyAttributes_3 Source)
  (connect CopyAttributes_2 Out CopyAttributes_3 Target)
  (connect AttributeFilter_2 InsideFilter CopyAttributes_4 Source)
  (connect AttributeFilter_2 InsideFilter CopyAttributes_4 Target)
  (connect Select_0 Out CreatePointsGrid_1 CellSize)
  (connect Select_1 Out CreatePointsGrid_1 GridExtents)
  (connect DefaultInputNode In CreatePolygon2D_0 In)
  (connect SortAttributes_0 Out CreateSpline_0 In)
  (connect Branch_2 Output%20B CreateSpline_1 In)
  (connect Branch_2 Output%20A CreateSpline_2 In)
  (connect FilterElementsByIndex_1 In%20Filter CreateSpline_3 In)
  (connect ToPoint_7 Out CreateSpline_4 In)
  (connect PointToAttributeSet_0 Out DeleteAttributes_0 In)
  (connect AddTags_2 Out FilterDataByIndex_0 In)
  (connect AddTags_3 Out FilterDataByIndex_1 In)
  (connect AttributeTransformOp_5 Out FilterDataByType_0 In)
  (connect Gather_0 Out FilterElementsByIndex_1 In)
  (connect AttributeFilter_0 OutsideFilter Gather_0 In)
  (connect AttributeFilter_1 OutsideFilter Gather_0 In)
  (connect CombinePoints_2 Out GetAttributeFromPointIndex_1 In)
  (connect CombinePoints_2 Out GetAttributeFromPointIndex_6 In)
  (connect CopyAttributes_0 Out GetElementsCount_0 In)
  (connect FilterDataByType_0 InsideFilter GetSplineControlPoints_2 In)
  (connect CreateSpline_3 Out Loop_0 In)
  (connect Loop_0 Out Loop_1 In)
  (connect CreateConstant_1 Out MakeRotatorAttribute_1 Pitch)
  (connect CreateConstant_1 Out MakeRotatorAttribute_1 Roll)
  (connect CreateConstant_1 Out MakeRotatorAttribute_1 Yaw)
  (connect AttributeFilter_3 InsideFilter MergePoints_0 In)
  (connect AttributeFilter_4 InsideFilter MergePoints_0 In)
  (connect TransformPoints_0 Out MergePoints_0 In)
  (connect CopyAttributes_4 Out PointToAttributeSet_0 In)
  (connect Branch_1 Output%20A Polygon2DOperation_2 ClipPaths)
  (connect CreateSpline_0 Out Polygon2DOperation_2 ClipPaths)
  (connect AttributeTransformOp_4 Out Polygon2DOperation_2 In)
  (connect MakeVectorAttribute_0 Out Select_0 Input%20A)
  (connect AttributeMathsOp_3 Out Select_0 Input%20B)
  (connect AttributeMathsOp_0 Out Select_1 Input%20A)
  (connect AttributeMathsOp_5 Out Select_1 Input%20B)
  (connect AttributePartition_0 Out SortAttributes_0 In)
  (connect WriteDataIndex_0 Out SplineSampler_1 Spline)
  (connect AttributeCast_1 Out ToPoint_0 In)
  (connect AttributeTransformOp_2 Out ToPoint_7 In)
  (connect AttributeFilterRange_0 InsideFilter TransformPoints_0 In)
  (connect MakeVectorAttribute_3 Out TransformPoints_0 OffsetMax)
  (connect AttributeMathsOp_4 Out TransformPoints_0 OffsetMin)
  (connect Branch_1 Output%20B WriteDataIndex_0 In)
  (output Out AttributeMathsOp_1)
  (output Out AttributeMathsOp_2))
