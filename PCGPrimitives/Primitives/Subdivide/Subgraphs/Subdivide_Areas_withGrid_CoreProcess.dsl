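; PCG subgraph that is the core process for subdividing spline-defined areas with a grid.
; It filters the input to spline data and feeds it to a Loop node running Subdivide_Areas_withGridLoop, which outputs the result.
; All parameters pass straight through to that loop: grid sizes or subdivision counts, angle jitter, and optional disturbance controls.
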(graph Subdivide_Areas_withGrid_CoreProcess
  (params (GridWidth :type Double :default 5e+03) (GridLength :type Double :default 5e+03) (Subdiv_width :type Int :default 1) (Subdiv_length :type Int :default 1) (Use_Subdivision_instead_of_sizes :type Bool) (AngleJitter :type Double) (Allow_Disturbance :type Bool) (Disturbance :type Double) (Disturbance_Subdivisions_distance :type Double :default 4e+03))
  ;[0 0 0 0.5] Core Process
  (bind FilterDataByType_1 (Filter%20Data%20By%20Type :TargetType (struct :Ids [(struct :Struct /Script/PCG.PCGDataTypeInfoSpline)])))
  (bind Loop_1 (Loop :Graph /PCGPrimitives/Primitives/Subdivide/Subgraphs/Subdivide_Areas_withGridLoop.Subdivide_Areas_withGridLoop :Allow_Disturbance Allow_Disturbance :AngleJitter AngleJitter :Disturbance Disturbance :Disturbance_Subdivisions_distance Disturbance_Subdivisions_distance :GridLength GridLength :GridWidth GridWidth :Subdiv_length Subdiv_length :Subdiv_width Subdiv_width :Use_Subdivision_instead_of_sizes Use_Subdivision_instead_of_sizes))
  (connect DefaultInputNode In FilterDataByType_1 In)
  (connect FilterDataByType_1 InsideFilter Loop_1 In)
  (output Out Loop_1))
