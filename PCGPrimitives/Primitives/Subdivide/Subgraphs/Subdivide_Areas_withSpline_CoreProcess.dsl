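; PCG core subgraph that subdivides areas using splines, returning both the cut areas and the clipped dividing splines.
; Data tagged Input:0 supplies areas (polygons, or splines forced into closed polygons), while the remaining splines act as cutters.
; A CutWithPaths polygon operation yields pieces rebuilt as closed linear splines, as Clip Paths trims the cutters against the original areas.
; Both outputs get fresh Type tags (ClosedSpline or Spline) and rotation forced to zero.
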(graph Subdivide_Areas_withSpline_CoreProcess
  (bind FilterDataByTag_0 (Filter%20Data%20By%20Tag :SelectedTags "Input:0"))
  (bind FilterDataByType_2 (Filter%20Data%20By%20Type :TargetType (struct :Ids [(struct :Struct /Script/PCG.PCGDataTypeInfoSpline)])))
  (bind CreatePolygon2D_0 (Create%20Polygon%202D :InputType ForceClosed)) ; Force closed polygon
  (bind FilterDataByType_0 (Filter%20Data%20By%20Type :TargetType (struct :Ids [(struct :Struct /Script/PCG.PCGDataTypeInfoPolygon2D)])))
  (bind FilterDataByType_1 (Filter%20Data%20By%20Type :TargetType (struct :Ids [(struct :Struct /Script/PCG.PCGDataTypeInfoSpline)])))
  (bind CreateConstant_1 (Create%20Constant))
  (bind MakeRotatorAttribute_1 (Make%20Rotator%20Attribute :Operation MakeRotFromAngles))
  ;[0.15 0.15 0.15 0.5] Areas
  (bind Polygon2DOperation_0 (Polygon%20Operation :Operation CutWithPaths))
  (bind ToPoint_0 (To%20Point))
  (bind CreateSpline_0 (Create%20Spline :bClosedLoop true :bLinear true :InterpTypeAttribute None))
  (bind DeleteTags_1 (Delete%20Tags :SelectedTags "Type:Point, Type:Spline, Type:ClosedSpline, Type:Polygon2D, Type:Surface, Type:Texture2D, Type:Volume"))
  (bind AddTags_0 (Add%20Tags :TagsToAdd "Type:ClosedSpline"))
  ;[0.15 0.15 0.15 0.5] Sudividing splines
  (bind ClipPaths_0 (Clip%20Paths))
  (bind FilterDataByType_3 (Filter%20Data%20By%20Type :TargetType (struct :Ids [(struct :Struct /Script/PCG.PCGDataTypeInfoSpline)])))
  (bind DeleteTags_2 (Delete%20Tags :SelectedTags "Type:Point, Type:Spline, Type:ClosedSpline, Type:Polygon2D, Type:Surface, Type:Texture2D, Type:Volume"))
  (bind AddTags_1 (Add%20Tags :TagsToAdd "Type:Spline"))
  ;[0.15 0.15 0.15 0.5] Force rotation 0
  (bind AttributeMathsOp_1 (Attribute%20Maths%20Op :InputSource1 $Rotation :Operation Set))
  (bind AttributeMathsOp_2 (Attribute%20Maths%20Op :InputSource1 $Rotation :Operation Set))
  (connect DeleteTags_1 Out AddTags_0 In)
  (connect DeleteTags_2 Out AddTags_1 In)
  (connect AddTags_0 Out AttributeMathsOp_1 InA)
  (connect MakeRotatorAttribute_1 Out AttributeMathsOp_1 InB)
  (connect AddTags_1 Out AttributeMathsOp_2 InA)
  (connect MakeRotatorAttribute_1 Out AttributeMathsOp_2 InB)
  (connect CreatePolygon2D_0 Out ClipPaths_0 ClipPolygons)
  (connect FilterDataByType_0 InsideFilter ClipPaths_0 ClipPolygons)
  (connect FilterDataByType_1 InsideFilter ClipPaths_0 In)
  (connect FilterDataByType_2 InsideFilter CreatePolygon2D_0 In)
  (connect ToPoint_0 Out CreateSpline_0 In)
  (connect CreateSpline_0 Out DeleteTags_1 In)
  (connect FilterDataByType_3 InsideFilter DeleteTags_2 In)
  (connect DefaultInputNode In FilterDataByTag_0 In)
  (connect FilterDataByTag_0 InsideFilter FilterDataByType_0 In)
  (connect FilterDataByTag_0 OutsideFilter FilterDataByType_1 In)
  (connect FilterDataByTag_0 InsideFilter FilterDataByType_2 In)
  (connect ClipPaths_0 Out FilterDataByType_3 In)
  (connect CreateConstant_1 Out MakeRotatorAttribute_1 Pitch)
  (connect CreateConstant_1 Out MakeRotatorAttribute_1 Roll)
  (connect CreateConstant_1 Out MakeRotatorAttribute_1 Yaw)
  (connect FilterDataByType_1 InsideFilter Polygon2DOperation_0 ClipPaths)
  (connect CreatePolygon2D_0 Out Polygon2DOperation_0 In)
  (connect FilterDataByType_0 InsideFilter Polygon2DOperation_0 In)
  (connect Polygon2DOperation_0 Out ToPoint_0 In)
  (output Out AttributeMathsOp_1)
  (output Out AttributeMathsOp_2))
